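# Translucent, two-sided unlit flame material for the Hang Spec glider effect, usable on Niagara mesh particles.
# Two panning, distorted noise textures, offset by the dynamic Random Offset, blend inside a sine-based frame shape.
# Rounded sine split bands and a distorted detail texture break up that mask, which drives both color and opacity.
# Emissive blends bright and inner colors with time-of-day tinting; opacity uses depth fade except on mobile.
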Material /BRCosmetics/Gadgets/Assets/VinderTech_GliderChute/Glider_HangSpec/FX/M_Glider_HangSpec_Flame
graph {
 properties {
  bUseAlphaToCoverage=true
  bUsedWithNiagaraMeshParticles=true
  blendMode=BLEND_Translucent
  pixelDepthOffsetMode=PDOM_Legacy
  shadingModel=MSM_Unlit
  twoSided=true
  usedShadingModels=Unlit
 }
 nodes [
  n1 = ComponentMask(None=TextureCoordinate()) {g=true}
  n2 = Multiply(A=@n1) {constB=2}
  Clamp(None=Multiply(A=Subtract(A=@n2)))
  n3 = VectorParameter() {group=Emissive, parameterName=Color Bright, defaultValue=<rgba 5 2.7786 2.2443 1>}
  n4 = VectorParameter() {group=Emissive, parameterName=Color Inner, defaultValue=<rgba 3 1.1415 1.1086 1>}
  n5 = MaterialFunctionCall() {materialFunction=&LinearGradient}
  LinearInterpolate(A=@n3.RGB, B=@n4.RGB, Alpha=@n5.VGradient)
  LinearInterpolate(A=@n3.RGB, B=@n4.RGB, Alpha=@n5.VGradient)
  n6 = MaterialFunctionCall(Float2=TextureCoordinate()) {materialFunction=&BreakOutFloat2Components}
  n7 = Multiply(A=ScalarParameter() {defaultValue=1.4, group=Frame Split, parameterName=Frame Split Offset A}, B=@n6.G) {constB=1.5}
  n8 = Multiply(A=OneMinus(None=Sine(None=@n7) {period=2}), B=ScalarParameter() {defaultValue=20, group=Frame Split, parameterName=Frame Split Thin A})
  n9 = DynamicParameter() {defaultValue=<rgba 1 1 1 100>, paramNames [Random Offset, Dissolve Progress, WPO Wave Intensity, Sphere Mask Range]}
  n10 = VectorParameter() {group=Frame Split, parameterName=Frame Split Distortion UV, defaultValue=<rgba 1 1 -2 0>}
  n11 = TextureCoordinate()
  n12 = Multiply(A=@n11, B=AppendVector(A=@n10.R, B=@n10.R))
  n13 = AppendVector(A=@n10.B, B=@n10.B)
  n14 = TextureSample(UVs=Add(A=AppendVector(A=@n9.Random Offset, B=@n9.Random Offset), B=Panner(Coordinate=@n12, Speed=@n13))) {samplerType=SAMPLERTYPE_Normal, texture=&T_Weap_Distort_N_02}
  n15 = ScalarParameter() {defaultValue=1.5, group=Frame Split, parameterName=Frame Split Distortion Intensity}
  n16 = ComponentMask(None=TextureCoordinate()) {r=true}
  n17 = LinearInterpolate(Alpha=Multiply(A=Multiply(A=@n14.R, B=@n15), B=OneMinus(None=@n16))) {constA=0.5, constB=1.5}
  n18 = ScalarParameter() {defaultValue=4, group=Frame Split, parameterName=Frame Split Depth A}
  n19 = Round(None=Add(A=Power(Base=@n8, Exp=@n17) {constExponent=0.5}, B=Power(Base=@n6.R, Exp=@n18)))
  n20 = Multiply(A=ScalarParameter() {defaultValue=0.8, group=Frame Split, parameterName=Frame Split Offset B}, B=@n6.G) {constB=1.5}
  n21 = Multiply(A=OneMinus(None=Sine(None=@n20) {period=2}), B=ScalarParameter() {defaultValue=15, group=Frame Split, parameterName=Frame Split Thin B})
  n22 = ScalarParameter() {defaultValue=3, group=Frame Split, parameterName=Frame Split Depth B}
  n23 = Round(None=Add(A=Power(Base=@n21, Exp=@n17) {constExponent=0.5}, B=Power(Base=@n6.R, Exp=@n22)))
  n24 = LinearInterpolate(
   B=Multiply(A=Saturate(None=@n19), B=Saturate(None=@n23))
   Alpha=ScalarParameter() {defaultValue=0.3, group=Frame Split, parameterName=Lerp Frame Split}
  ) {constA=1, constB=0}
  n25 = VectorParameter() {group=Frame Detail, parameterName=Frame Detail Distortion UV, defaultValue=<rgba 0.5 2 -1 0.25>}
  n26 = Multiply(A=@n11, B=AppendVector(A=@n25.R, B=@n25.R))
  n27 = AppendVector(A=@n25.B, B=@n25.B)
  n28 = TextureSample(UVs=Add(A=AppendVector(A=@n9.Random Offset, B=@n9.Random Offset), B=Panner(Coordinate=@n26, Speed=@n27))) {samplerType=SAMPLERTYPE_Normal, texture=&T_Weap_Distort_N_02}
  n29 = ScalarParameter() {defaultValue=0.008, group=Frame Detail, parameterName=Frame Detail Distortion Intensity}
  n30 = ComponentMask(None=TextureCoordinate()) {g=true}
  n31 = TextureSampleParameter2D(UVs=Multiply(A=Multiply(A=@n28.G, B=@n29), B=OneMinus(None=@n30))) {
   group=Frame Detail
   parameterName=Frame Detail Tex
   texture=/Game/Gadgets/Assets/VinderTech_GliderChute/Glider_Headset/FX/T_Glider_Headset_Mask_01
  }
  n32 = Power(Base=@n31.G, Exp=ScalarParameter() {defaultValue=2, group=Frame Detail, parameterName=Frame Detail Power})
  n33 = ScalarParameter() {defaultValue=5, group=Frame Detail, parameterName=Frame Detail Mult}
  n34 = LinearInterpolate(
   B=Saturate(None=Multiply(A=@n32, B=@n33) {constB=8})
   Alpha=ScalarParameter() {defaultValue=0.4, group=Frame Detail, parameterName=Lerp Frame Detail}
  ) {constA=1, constB=0}
  n35 = DynamicParameter() {defaultValue=<rgba 1 1 1 100>, paramNames [Random Offset, Dissolve Progress, WPO Wave Intensity, Sphere Mask Range]}
  n36 = VectorParameter() {group=Noise, parameterName=Noise B UV, sortPriority=6, defaultValue=<rgba 0.5 1 -1 0.2>}
  n37 = AppendVector(A=@n36.R, B=@n36.R)
  n38 = TextureCoordinate()
  n39 = Add(
   A=AppendVector(A=@n35.Random Offset, B=@n35.Random Offset)
   B=Panner(Coordinate=Multiply(A=@n38, B=@n37), Speed=AppendVector(A=@n36.B, B=@n36.B))
  )
  n40 = VectorParameter() {group=Distortion, parameterName=Distortion UV, sortPriority=1, defaultValue=<rgba 1 1 -3 0>}
  n41 = Multiply(A=@n38, B=AppendVector(A=@n40.R, B=@n40.R))
  n42 = AppendVector(A=@n40.B, B=@n40.B)
  n43 = TextureSampleParameter2D(UVs=Add(A=AppendVector(A=@n35.Random Offset, B=@n35.Random Offset), B=Panner(Coordinate=@n41, Speed=@n42))) {
   group=Distortion
   parameterName="Distortion  Tex"
   samplerType=SAMPLERTYPE_Normal
   sortPriority=0
   texture=&T_Weap_Distort_N_02
  }
  n44 = OneMinus(None=ComponentMask(None=@n38) {g=true})
  n45 = ScalarParameter() {defaultValue=0.5, group=Distortion, parameterName=Distortion Gradient Lerp, sortPriority=3}
  n46 = Multiply(
   A=Multiply(A=ComponentMask(None=@n43.RGB) {g=true, r=true}, B=LinearInterpolate(B=@n44, Alpha=@n45) {constA=1, constAlpha=0})
   B=ScalarParameter() {defaultValue=0.2, group=Distortion, parameterName=Distortion Intensity, sortPriority=2}
  )
  n47 = ChannelMaskParameter(None=TextureSampleParameter2D(UVs=Add(A=@n39, B=@n46)) {group=Noise, parameterName=NoIse B Texture, sortPriority=4, texture=&T_NoisePack_01}.RGB) {group=Noise, maskChannel=Green, parameterName=Noise B Channel, sortPriority=5, defaultValue=<rgba 0 1 0 0>}
  n48 = Saturate(None=Multiply(A=@n47.R, B=ScalarParameter() {defaultValue=1, group=Noise, parameterName=NoIse B Intensity, sortPriority=7}) {constB=10})
  n49 = AppendVector(A=@n35.Random Offset, B=@n35.Random Offset)
  n50 = VectorParameter() {group=Noise, parameterName=Noise A UV, sortPriority=2, defaultValue=<rgba 1 3 -2 -0.4>}
  n51 = Panner(Coordinate=Multiply(A=@n38, B=AppendVector(A=@n50.R, B=@n50.R)), Speed=AppendVector(A=@n50.B, B=@n50.B))
  n52 = TextureSampleParameter2D(UVs=Add(A=@n46, B=Add(A=@n49, B=@n51))) {
   group=Noise
   parameterName=NoIse A Texture
   sortPriority=0
   texture=/BRCosmetics/Weapons/FORT_Melee/Pickaxe_Headset/FX/T_Pickaxe_Headset_Noise_01
  }
  n53 = Saturate(None=Multiply(
    A=ChannelMaskParameter(None=@n52.RGB) {group=Noise, parameterName=Noise A Channel, sortPriority=1}.R
    B=ScalarParameter() {defaultValue=1, group=Noise, parameterName=NoIse A Intensity, sortPriority=3}
   ) {constB=10})
  n54 = ScalarParameter() {defaultValue=0.5, group=Noise, parameterName=Lerp Noise AB, sortPriority=8}
  n55 = MaterialFunctionCall(Float2=TextureCoordinate()) {materialFunction=&BreakOutFloat2Components}
  n56 = Multiply(A=@n55.G) {constB=0.5}
  n57 = ScalarParameter() {defaultValue=3, group=Frame Shape, parameterName=Top Thickness, sortPriority=2}
  n58 = ScalarParameter() {defaultValue=0.5, group=Frame Shape, parameterName=Bottom Thickness, sortPriority=3}
  n59 = OneMinus(None=@n55.R)
  n60 = Multiply(A=Power(Base=Sine(None=@n56), Exp=LinearInterpolate(A=@n57, B=@n58, Alpha=@n59) {constA=6, constB=2}) {constExponent=4}, B=@n59)
  n61 = Power(Base=@n60)
  n62 = Multiply(A=Add(A=LinearInterpolate(A=@n48, B=@n53, Alpha=@n54), B=Saturate(None=@n61)), B=@n60)
  n63 = LinearInterpolate(
   A=ScalarParameter() {defaultValue=1.25, group=Frame Shape, parameterName=Top Fade, sortPriority=0}
   B=ScalarParameter() {defaultValue=1, group=Frame Shape, parameterName=Bottom Fade, sortPriority=1}
   Alpha=@n59
  ) {constA=1, constB=0.5}
  n64 = Multiply(A=Multiply(A=Multiply(A=@n24, B=@n34), B=Power(Base=@n62, Exp=@n63)), B=ComponentMask(None=TextureCoordinate()) {r=true})
  n65 = Multiply(A=MaterialFunctionCall(1=@n3.RGB, 2=@n4.RGB, 3=@n4.RGB, 4=@n3.RGB, A=@n1) {materialFunction=&Lerp_Multiple_Float3}.Lerp 4 Inputs)
  n66 = Multiply(A=Add(A=@n64, B=Multiply(A=@n64, B=@n65)), B=Add(A=@n4.RGB, B=Power(Base=@n65) {constExponent=1}))
  n67 = ScalarParameter() {defaultValue=0.75, group=Emissive, parameterName=Emissive_Intensity}
  n68 = Constant3Vector() {constant=<rgba 1 1 1 0>}
  n69 = Constant3Vector() {constant=<rgba 1.5 1.5 1.5 0>}
  n70 = Constant3Vector() {constant=<rgba 1 1 2 0>}
  MaterialFunctionCall(Emissive=Multiply(
    A=Multiply(A=@n66, B=@n67)
    B=MaterialFunctionCall(Color_Morning=@n68, Color_Day=@n69, Color_Evening=@n68, Color_Night=@n70) {materialFunction=&Color_Time_of_Day}.Result
   )) {materialFunction=&MF_FX_EmissiveAdjustments}
  n71 = TextureCoordinate()
  n72 = Multiply(
   A=SmoothStep(Value=ComponentMask(None=@n71) {r=true}) {constMax=0.3, constMin=0.05}
   B=ScalarParameter() {defaultValue=100, group=Opacity, parameterName=DepthFadeDistance}
  )
  n73 = ScalarParameter() {defaultValue=6, group=Opacity, parameterName=Opacity Multiplier, sortPriority=1}
  ShadingPathSwitch(Default=Multiply(A=DepthFade(Opacity=@n64, FadeDistance=@n72), B=@n73) {constB=3}, Mobile=Multiply(A=@n64, B=@n73)) {desc=Migrated from FeatureLevelSwitch}
 ]
}
