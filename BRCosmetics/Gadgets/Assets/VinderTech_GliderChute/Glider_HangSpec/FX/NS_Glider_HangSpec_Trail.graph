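# Niagara trail effect for the HangSpec glider chute, combining socket-attached thruster flames, flares, embers, speedlines and bone-driven ribbon trails.
# A system timeline ramps flame scale, ember spawning and speed, and trail and speedline opacity, while a smoothed system value steers ember gravity.
# User.Variant blends ember and glow colors, forward/right dot inputs skew speedlines, and front-end flags flip directions for preview.
# Flames and flares also emit lights.
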NiagaraSystem /BRCosmetics/Gadgets/Assets/VinderTech_GliderChute/Glider_HangSpec/FX/NS_Glider_HangSpec_Trail
dependencies {
 usedDataInterfaces [
  &NiagaraDataInterfaceCurve
  &NiagaraDataInterfaceArrayFloat
  &NiagaraDataInterfaceArrayInt32
  &NiagaraDataInterfaceArrayFloat3
  &NiagaraDataInterfaceColorCurve
  &NiagaraDataInterfaceVector2DCurve
  &NiagaraDataInterfaceSocketReader
  &NiagaraDataInterfaceArrayColor
 ]
 usedDynamicInputs [
  &MaskVectorBySpawnGroup
  &Boolean_AndOperation
  &SetBoolByFloatComparison
  &NormalizeVector
  &TransformVector
  &SelectLinearColorFromArray
  &MakeIntFromFloat
 ]
 usedModules [
  &self:SystemSpawnScript.SystemScriptSource.SystemScriptGraph.NiagaraNodeAssignment_0.SetVariables_#1
  &self:SystemSpawnScript.SystemScriptSource.SystemScriptGraph.NiagaraNodeAssignment_3.SetVariables_#2
  &SystemState
  &Timeline
  &FN_Integrator
  &InterpolateOverTime
  &self:SystemSpawnScript.SystemScriptSource.SystemScriptGraph.NiagaraNodeAssignment_1.SetVariables_#3
  &self:SystemSpawnScript.SystemScriptSource.SystemScriptGraph.NiagaraNodeAssignment_2.SetVariables_#4
 ]
}
emitters {
 Embers002 {
  emitterUpdateScript [
   &EmitterState(Life Cycle Mode=::NewEnumerator1, Loop Behavior=::NewEnumerator1, Loop Delay=1, Loop Duration Mode=::NewEnumerator1, UseLoopDelay=true)
   &SpawnBurst_Instantaneous(Spawn Count=50, Use Spawn Probability=false, UseLoopCountLimit=false)
   &SpawnRate(
    SpawnRate=&FloatFromCurve(FloatCurve=DataInterface() {propertyValues {Curve {keys [<at 0 1>,<at 2 0>]}}}, CurveIndex=$Emitter.Age, Scale Curve=100)
    Use Spawn Probability=true
   )
   &SpawnRate(
    Spawn Probability=&FloatFromCurve(
     FloatCurve=DataInterface() {propertyValues {Curve {keys [
         {arriveTangent=3.3514, interpMode=RCIM_Cubic, leaveTangent=3.3514, tangentMode=RCTM_User, time=0, value=0}
         {interpMode=RCIM_Cubic, tangentMode=RCTM_User, time=1, value=1}
        ]}}}
     CurveIndex=$System.Timeline.Time
    )
    SpawnRate=100
    Use Spawn Probability=true
   ) {module=SpawnRate001}
  ]
  particleSpawnScript [
   &PartitionParticles()
   /SetVariables(Particles.PerParticleRandom:NiagaraFloat=&RandomRangeFloat(Minimum=0, Maximum=1)) {isSetParameters=true}
   &FN_SetSpawnGroup(Number of Spawn Groups=3) {enabled=false}
   &InitializeParticle(
    Lifetime=&Multiply_Float(
     A=&Lerp_Float(A=0.2, B=0.4, Alpha=$Particles.PerParticleRandom)
     B=&Lerp_Float(A=0.8, B=0.75, Alpha=$System.Timeline.Time)
    )
    Mesh Scale=<2 1 1>
    Mesh Scale Mode=::NewEnumerator4
    Sprite Size=&Multiply_Vector2D(
     A=&MergeHelper_Vector2Switch(0=<1 1.5>, 1=<0.5 0.75>, 2=<0.5 0.75>, 3=?, 4=?, Index=&SpawnGroup(), Num=::NewEnumerator1)
     B=&MakeVector2D(
      X=&FloatFromCurve(
       FloatCurve=DataInterface() {propertyValues {Curve {keys [<at 0 1>,<at 0.8 2 RCIM_Constant>,<at 0.95 3 RCIM_Constant>]}}}
       CurveIndex=$Particles.PerParticleRandom
      )
      Y=&FloatFromCurve(
       FloatCurve=DataInterface() {propertyValues {Curve {keys [<at 0 1>,<at 0.8 2 RCIM_Constant>,<at 0.95 3>]}}}
       CurveIndex=$Particles.PerParticleRandom
      )
     )
    )
    Sprite Size Mode=::NewEnumerator4
    UsePositionOffset=false
    Write Material Random=true
   )
   &SocketLocation(
    SamplePosOffset=false
    Socket Index=$Particles.PartitionParticles.Partition
    Socket Reader=$System.Socket Reader
    Socket Sampling Mode=::NewEnumerator3
    VelocitySpeedLimit=false
   )
   &ShapeLocation(
    MaskSpawnGroup=false
    Override Seed=false
    Rotation Axis=&FN_FrontEndVector3Modifier(Gameplay=<0 1 0>, Front End=<0 -1 0>)
    Rotation Mode=::NewEnumerator1
    Sphere Radius=&MergeHelper_FloatSwitch(0=3, 1=0.2, 2=0.2, 3=?, Index=&SpawnGroup(), Num=::NewEnumerator1)
    Sphere Surface Distribution=1
    UseOwnerScale=true
   )
   &AddVelocity(
    Velocity=&Add_Vector(
     A=$Particles.ShapeLocation.ShapeNormal
     B=&FN_FrontEndVector3Modifier(Gameplay=<0 3 0>, Front End=<0 -3 0>)
    )
    Velocity Speed Scale=&Multiply_Float(
     A=&Lerp_Float(A=50, B=75, Alpha=$Particles.PerParticleRandom)
     B=&Multiply_Float(
      A=&MergeHelper_FloatSwitch(
       0=&Lerp_Float(A=2, B=3, Alpha=$System.Timeline.Time)
       1=&Lerp_Float(A=1, B=2, Alpha=$System.Timeline.Time)
       2=&Lerp_Float(A=1, B=2, Alpha=$System.Timeline.Time)
       3=?
       Index=&SpawnGroup()
       Num=::NewEnumerator1
      )
      B=&FloatFromCurve(FloatCurve=DataInterface() {propertyValues {Curve {keys [<at 2 2.5 RCIM_Cubic>,<at 2.5 1 RCIM_Cubic>]}}}, CurveIndex=$Emitter.Age)
     )
    )
   )
  ]
  particleUpdateScript [
   &ParticleState()
   &Drag(
    Drag=&FloatFromCurve(
     FloatCurve=DataInterface() {propertyValues {Curve {keys [
         {interpMode=RCIM_Cubic, tangentMode=RCTM_User, time=0, value=0.8}
         {arriveTangent=0.4778, interpMode=RCIM_Cubic, leaveTangent=0.4778, tangentMode=RCTM_User, time=1, value=1.0375}
        ]}}}
     Scale Curve=5
    )
    Use Linear Drag=true
    Use Rotational Drag=false
   )
   &CurlNoiseForce(Noise Frequency=100, Noise Strength=2000, Pan Noise=true, Pan Noise Field=<0 0 1>) {module=CurlNoiseForce001}
   &GravityForce(
    Coordinate Space=::Simulation
    Gravity=&Multiply_VectorByFloat(
     Vector=<1 0 0>
     Float=&Multiply_Float(
      A=&Multiply_Float(
       A=&Lerp_Float(A=250, B=375, Alpha=$System.Timeline.Time)
       B=&Lerp_Float(A=1, B=1.5, Alpha=$Particles.PerParticleRandom)
      )
      B=$System.InterpolateOverTime001.InterpolateOverTime_Float.Moving Average
     )
    )
   )
   &SolveForcesAndVelocity(Clamp Velocity=false, Limit Acceleration=false)
   &ScaleColor(
    Scale Alpha=&FloatFromCurve(FloatCurve=DataInterface() {propertyValues {Curve {keys [{interpMode=RCIM_Cubic, tangentMode=RCTM_User, time=0, value=1}, {interpMode=RCIM_Cubic, tangentMode=RCTM_User, time=1, value=0}]}}})
    Scale RGB=&Lerp_Vector(
     A=&MakeVectorFromLinearColorRGB(Color=&ColorFromCurve(ColorCurve=DataInterface() {propertyValues {
         AlphaCurve {keys [<at 0 1 RCIM_Cubic>]}
         BlueCurve {keys [<at 0.0088 0>,<at 0.3924 0.3754 RCIM_Cubic>]}
         GreenCurve {keys [<at 0.0088 1.567>,<at 0.3924 0 RCIM_Cubic>]}
         RedCurve {keys [<at 0.0088 3>,<at 0.3924 1 RCIM_Cubic>]}
        }}))
     B=&MakeVectorFromLinearColorRGB(Color=&ColorFromCurve(ColorCurve=DataInterface() {propertyValues {
         AlphaCurve {keys [<at 0 1 RCIM_Cubic>]}
         BlueCurve {keys [<at 0.3321 5>,<at 0.4483 0.3754 RCIM_Cubic>]}
         GreenCurve {keys [<at 0.3321 0.5225>,<at 0.4483 0 RCIM_Cubic>]}
         RedCurve {keys [<at 0.3321 0>,<at 0.4483 1 RCIM_Cubic>]}
        }}))
     Alpha=$User.Variant
    )
    ScaleA=true
    ScaleRGB=true
   )
   &ScaleSpriteSize(
    Uniform Curve Scale=&MakeFloatFromVector(Vector=$Engine.Owner.Scale, Channel=::NewEnumerator0)
    Uniform Curve Sprite Scale=DataInterface() {propertyValues {Curve {keys [
        {tangentMode=RCTM_User, time=0, value=1}
        {arriveTangent=-0.0405, interpMode=RCIM_Cubic, leaveTangent=-0.0405, tangentMode=RCTM_User, time=0.7009, value=1}
        {arriveTangent=-8.0089, interpMode=RCIM_Cubic, leaveTangent=-8.0089, tangentMode=RCTM_User, time=1, value=0}
       ]}}}
   )
   &ScaleSpriteSizeBySpeed(
    Max Scale Factor=<0.5 4>
    Sample Scale Factor By Curve=true
    Scale Factor Curve=DataInterface() {propertyValues {Curve {keys [<at 0 0>, {arriveTangent=1.4989, interpMode=RCIM_Cubic, tangentMode=RCTM_Break, time=1, value=1}]}}}
    Velocity Threshold=250
   )
  ]
  rendererClasses [&NiagaraSpriteRendererProperties]
  renderers [{rendererClass=&NiagaraSpriteRendererProperties}]
 }
 FN_AnimTrail {
  emitterSpawnScript [&FN_RibbonIDOffsetUtility()]
  emitterUpdateScript [
   &EmitterState()
   &FN_SpawnAnimTrail(Number of Anim Trails=3)
  ]
  particleSpawnScript [
   &FN_InitializeAnimTrailBoneIndecies(First Bone or Socket Index=4, Second Bone or Socket Index=3)
   &FN_InitializeAnimTrailBoneIndecies(AnimTrail Selection=1, First Bone or Socket Index=5, Second Bone or Socket Index=4) {module=FN_InitializeAnimTrailBoneIndecies001}
   &FN_InitializeAnimTrailBoneIndecies(AnimTrail Selection=2, First Bone or Socket Index=7, Second Bone or Socket Index=6) {module=FN_InitializeAnimTrailBoneIndecies002}
   &FN_InitializeAnimTrail(Color=&MaskLinearColorBySpawnGroup(Masked Linear Color=<rgba 1 0.145 0.2992 1>, Passthrough Linear Color=<rgba 0.0459 0.0459 0.0459 2>))
   &AddVelocity(Velocity=<0 150 0>) {enabled=false}
   &AddVelocity(
    Velocity=<0 -100 0>
    Velocity Speed Scale=&MakeCustomFloatFromBool(Bool=$User.bIsFrontEndPreview, True Float=1, False Float=-1.5)
   ) {module=AddVelocity001}
  ]
  particleUpdateScript [
   &SolveForcesAndVelocity(Clamp Velocity=false, Limit Acceleration=false)
   &FN_UpdateAnimTrail() {module=FN_UpdateAnimTrail001}
   &ScaleColor(
    Scale Alpha=&RemapRange(
     Input Value To Remap=$System.Age
     Input Min=3.5
     Input Max=4.5
     Output Min=0
     Output Max=&Lerp_Float(A=0, B=1, Alpha=$System.Timeline.Time)
    )
    ScaleA=true
    ScaleRGB=true
   )
  ]
  rendererClasses [&NiagaraRibbonRendererProperties]
  renderers [
   {rendererClass=&NiagaraRibbonRendererProperties}
   {rendererClass=&NiagaraRibbonRendererProperties, rendererIndex=1}
  ]
 }
 Flames {
  emitterUpdateScript [
   &EmitterState(
    Inactive Response=::NewEnumerator1
    Life Cycle Mode=::NewEnumerator1
    Loop Behavior=::NewEnumerator1
    Loop Delay=0.25
    Loop Duration Mode=::NewEnumerator1
    UseLoopDelay=false
   )
   &SpawnBurst_Instantaneous(Spawn Count=4, Spawn Time=1, Use Spawn Probability=false, UseLoopCountLimit=false) {module=SpawnBurst_Instantaneous001}
  ]
  particleSpawnScript [
   &PartitionParticles(Number Of Partitions=4)
   &InitializeParticle(
    Color=$System.ColorA
    Mesh Scale=&MergeHelper_VectorSwitch(0=<1.5 0.6 0.5>, 1=<1 0.15 0.25>, 2=<1 0.15 0.25>, 3=?, Index=&SpawnGroup(), Num=::NewEnumerator1)
    Mesh Scale Mode=::NewEnumerator4
    Sprite Size Mode=::NewEnumerator3
    Uniform Sprite Size=5
    UsePositionOffset=false
    Write Material Random=true
   )
   &InitialMeshOrientation(Mesh Orientation Mode=::NewEnumerator2, Use Rotation Vector=false)
  ]
  particleUpdateScript [
   &ParticleState(Kill Particles When Lifetime Has Elapsed=false)
   &SocketLocation(
    SamplePosOffset=false
    Socket Index=&SelectIntFromArray(Direct Array Index=$Particles.PartitionParticles.Partition, Int Selection Array=DataInterface() {propertyValues {IntData [0,1,2,8]}})
    Socket Reader=$System.Socket Reader
    Socket Sampling Mode=::NewEnumerator3
    VelocitySpeedLimit=false
   )
   &ScaleMeshSize(Scale Factor=&Multiply_Vector(
     A=&SelectVectorFromArray(
      Direct Array Index=$Particles.PartitionParticles.Partition
      Vector Selection Array=DataInterface() {propertyValues {FloatData [<0.8 1 1>, <0.85 1 1>, <0.35 0.25 0.25>, <0.35 0.25 0.25>]}}
     )
     B=&MakeVector(
      X=&FloatFromCurve(
       FloatCurve=DataInterface() {propertyValues {Curve {keys [
           {arriveTangent=17.94, interpMode=RCIM_Cubic, leaveTangent=17.94, tangentMode=RCTM_User, time=0, value=0}
           {arriveTangent=-1.8497, interpMode=RCIM_Cubic, leaveTangent=-1.8497, tangentMode=RCTM_User, time=0.2, value=1.775}
           {arriveTangent=-0.0405, interpMode=RCIM_Cubic, leaveTangent=-0.0405, tangentMode=RCTM_User, time=1, value=1.25}
          ]}}}
       Scale Curve=&FloatFromCurve(
        FloatCurve=DataInterface() {propertyValues {Curve {keys [
            {arriveTangent=-0.8588, interpMode=RCIM_Cubic, time=0, value=0.8}
            {arriveTangent=3.1777, interpMode=RCIM_Cubic, leaveTangent=-2.1174, tangentMode=RCTM_Break, time=1, value=2}
           ]}}}
        CurveIndex=$System.Timeline.Time
        Scale Curve=&FloatFromCurve(
         FloatCurve=DataInterface() {propertyValues {Curve {
            defaultValue=0.45
            keys [
             {interpMode=RCIM_Cubic, tangentMode=RCTM_User, time=0, value=0}
             {interpMode=RCIM_Cubic, tangentMode=RCTM_User, time=0.25, value=1}
             {tangentMode=RCTM_User, time=0.6, value=0.3}
             <at 1 1 RCIM_Cubic>
            ]
           }}}
         Scale Curve=1.5
        )
       )
      )
      Y=1
      Z=1
     )
    ))
   &FN_Light_Attributes(
    Color=$Particles.Color
    Create Color Attribute=true
    Create Position Attribute=true
    Create Radius Attribute=true
    Position=&AddVectorToPosition(Position=$Particles.Position, Vector=<0 3 0>)
    Radius=30
    Scale Color=5
   )
  ]
  rendererClasses [&NiagaraMeshRendererProperties, &NiagaraLightRendererProperties]
  renderers [
   {rendererClass=&NiagaraMeshRendererProperties, rendererIndex=0}
   {rendererClass=&NiagaraMeshRendererProperties, rendererIndex=1}
   {rendererClass=&NiagaraLightRendererProperties, rendererIndex=2}
  ]
 }
 FlareLoop003 {
  emitterUpdateScript [
   &EmitterState(Delay First Loop Only=true, Life Cycle Mode=::NewEnumerator1, Loop Delay=0.5, Loop Duration=0.25, UseLoopDelay=true)
   &SpawnBurst_Instantaneous(Spawn Count=2, Use Spawn Probability=false, UseLoopCountLimit=false)
  ]
  particleSpawnScript [
   &PartitionParticles()
   &InitializeParticle(
    Color=$System.ColorB
    Lifetime=0.75
    Sprite Size Mode=::NewEnumerator3
    Uniform Sprite Size=15
    UsePositionOffset=false
    Write Material Random=true
   )
  ]
  particleUpdateScript [
   &ParticleState()
   &ScaleSpriteSize(
    Uniform Curve Index=$Emitter.Age
    Uniform Curve Scale=&SelectFloatFromArray(Direct Array Index=$Particles.PartitionParticles.Partition, Float Selection Array=DataInterface() {propertyValues {FloatData [1,1,0.25]}})
    Uniform Curve Sprite Scale=DataInterface() {propertyValues {Curve {keys [{tangentMode=RCTM_User, time=0, value=0}, {tangentMode=RCTM_User, time=0.75, value=1}]}}}
   )
   &ScaleColor(
    Scale Alpha=&FloatFromCurve(
     FloatCurve=DataInterface() {propertyValues {Curve {keys [
         {arriveTangent=64.5758, interpMode=RCIM_Cubic, leaveTangent=64.5758, tangentMode=RCTM_User, time=0, value=0}
         {arriveTangent=72.9722, interpMode=RCIM_Cubic, leaveTangent=72.9722, tangentMode=RCTM_User, time=0.1, value=1}
        ]}}}
     CurveIndex=$Emitter.Age
     Scale Curve=0.75
    )
    ScaleA=true
    ScaleRGB=true
   )
   &SocketLocation(
    SamplePosOffset=false
    Socket Index=&SelectIntFromArray(Direct Array Index=$Particles.PartitionParticles.Partition, Int Selection Array=DataInterface() {propertyValues {IntData [0,1]}})
    Socket Reader=$System.Socket Reader
    Socket Sampling Mode=::NewEnumerator3
    VelocitySpeedLimit=false
   )
   /SetVariables(
    Particles.Color_SoftGlow:LinearColor=&LerpLinearColors(StartColor=<rgba 1 0.4739 0.2876 0.5>, EndColor=<rgba 0.6079 0.3685 1 1>, LerpFactor=$User.Variant)
    Particles.Position_SoftGlow:NiagaraPosition=&AddVectorToPosition(Position=$Output.SocketLocation.SampledPosition, Vector=<0 2 0>)
    Particles.SpriteSize_SoftGlow:Vector2f=&Multiply_Vector2D_ByFloat(
     Float=&FloatFromCurve(
      FloatCurve=DataInterface() {propertyValues {Curve {keys [{tangentMode=RCTM_User, time=0, value=0}, {tangentMode=RCTM_User, time=0.75, value=1}]}}}
      CurveIndex=$Emitter.Age
      Scale Curve=&SelectFloatFromArray(Direct Array Index=$Particles.PartitionParticles.Partition, Float Selection Array=DataInterface() {propertyValues {FloatData [2,2,0.5,0.5]}})
     )
     Vector2D=<10 10>
    )
   ) {isSetParameters=true}
   &DynamicMaterialParameters(
    Index 0 Param 1=1
    Index 0 Param 2=2
    Index 0 Param 3=1
    Index 0 Param 4=1
    Module.Index0Float0Write=true
    Module.Index0Float1Write=true
    Module.Index0Float2Write=true
    Module.Index0Float3Write=true
   )
   &SpriteFacingAndAlignment(
    Sprite Facing=&FN_FrontEndVector3Modifier(Gameplay=<0 1 0>, Front End=<0 -1 0>)
    WriteAlignment=false
    WriteFacing=true
   )
   &FN_Light_Attributes(Color=$System.ColorB, Create Color Attribute=true, Create Exponent Attribute=true, Create Radius Attribute=true, Scale Color=3)
  ]
  rendererClasses [&NiagaraSpriteRendererProperties, &NiagaraLightRendererProperties]
  renderers [
   {rendererClass=&NiagaraSpriteRendererProperties}
   {rendererClass=&NiagaraSpriteRendererProperties, rendererIndex=1}
   {rendererClass=&NiagaraLightRendererProperties, rendererIndex=2}
  ]
 }
 Speedlines {
  emitterUpdateScript [
   &EmitterState()
   &SpawnRate(
    SpawnRate=&MakeCustomFloatFromBool(Bool=$User.bIsFrontEndPreview, True Float=90, False Float=90)
    Use Spawn Probability=true
   ) {module=SpawnRate001}
  ]
  particleSpawnScript [
   &InitializeParticle(
    Color=&MakeLinearColorFromVectorAndFloat(
     "Vector (RGB)"=<1 1 1>
     "Float (Alpha)"=&Multiply_Float(
      A=&UniformRangedFloat(Minimum=1, Maximum=1)
      B=&MakeCustomFloatFromBool(
       Bool=$User.bIsFrontEndPreview
       True Float=&FloatFromCurve(
        FloatCurve=DataInterface() {propertyValues {Curve {keys [{tangentMode=RCTM_User, time=0, value=0}, {tangentMode=RCTM_User, time=2, value=1}]}}}
        CurveIndex=$Emitter.Age
        Scale Curve=0.4
       )
       False Float=1
      )
     )
    )
    Lifetime Max=0.2
    Lifetime Min=0.15
    Lifetime Mode=::NewEnumerator1
    Position Offset=&FN_FrontEndVector3Modifier(Gameplay=<-50 0 0>, Front End=<50 0 0>)
    Sprite Size Max=<3 10>
    Sprite Size Min=<2 2>
    Sprite Size Mode=::NewEnumerator2
    UsePositionOffset=true
    Write Material Random=true
   )
   &AddVelocity(
    Module.Use Falloff Distance=false
    Origin Offset=&FN_FrontEndVector3Modifier(Gameplay=<0 300 0>, Front End=<0 -300 0>)
    Override Seed=false
    Velocity Mode=::NewEnumerator1
    Velocity Speed=-2500
   ) {module=AddVelocity001}
   &ShapeLocation(
    Handle Radius=0
    Large Radius=&FN_FrontEndFloatModifier(
     Gameplay=&RandomRangeFloat(Minimum=250, Maximum=400)
     Front End=&RandomRangeFloat(Minimum=300, Maximum=350)
    )
    MaskSpawnGroup=false
    Offset=&FN_FrontEndVector3Modifier(Gameplay=<0 -500 0>, Front End=<0 500 0>)
    Offset Mode=::NewEnumerator0
    Override Seed=false
    Rotation Angle Type:ENiagara_AngleInput=::NewEnumerator0
    Rotation Axis=<0 1 0>
    Rotation Mode=::NewEnumerator1
    Shape Primitive=::NewEnumerator3
    UseOwnerScale=true
   )
   &AddVelocity(
    Rotation Coordinate Space=::Simulation
    Velocity=&Add_Vector(
     A=&FN_FrontEndVector3Modifier(Gameplay=<0 0 0>, Front End=<0 0 0>)
     B=&Add_Vector(
      A=&Multiply_VectorByFloat(Vector=<550 0 0>, Float=$User.RightDot)
      B=&Multiply_VectorByFloat(Vector=<0 150 0>, Float=$User.ForwardDot)
     )
    )
    Velocity Speed Scale=&FN_FrontEndFloatModifier(Gameplay=1, Front End=1)
   )
  ]
  particleUpdateScript [
   &SolveForcesAndVelocity(Clamp Velocity=false, Limit Acceleration=false)
   &ParticleState()
   &ScaleSpriteSize(Scale Factor=&Vector2DFromCurve(
     Vector2Curve=DataInterface() {propertyValues {XCurve {keys [<at 0 0.5 RCIM_Cubic>,<at 0.5 1 RCIM_Cubic>]}, YCurve {keys [<at 0 0 RCIM_Cubic>,<at 0.5 1 RCIM_Cubic>]}}}
     ScaleCurve=&Vector2DFromFloat(Value=1)
    ))
   &ScaleSpriteSizeBySpeed(Max Scale Factor=<1 50>, Min Scale Factor=<0.01 0.05>, Velocity Threshold=1)
   &ScaleColor(
    Scale Alpha=&FloatFromCurve(
     FloatCurve=DataInterface() {propertyValues {Curve {keys [
         <at 0 0 RCIM_Cubic>
         <at 0.3759 1.0207>
         {arriveTangent=-2.221, interpMode=RCIM_Cubic, leaveTangent=-2.221, tangentMode=RCTM_User, time=1, value=0}
        ]}}}
     Scale Curve=&MakeCustomFloatFromBool(
      Bool=$User.bIsFrontEndPreview
      True Float=&MakeCustomFloatFromBool(Bool=$User.bIsFullyDeployed, True Float=0.5, False Float=0)
      False Float=&Lerp_Float(A=0, B=0.5, Alpha=&FN_FrontEndFloatModifier(Gameplay=$System.Timeline.Time, Front End=0.5))
     )
    )
    Scale RGB=&VectorFromFloat(Value=&FN_FrontEndFloatModifier(Gameplay=0.5, Front End=0.33))
    ScaleA=true
    ScaleRGB=true
   )
  ]
  rendererClasses [&NiagaraSpriteRendererProperties]
  renderers [{rendererClass=&NiagaraSpriteRendererProperties}]
 }
}
summary {
 systemName=NS_Glider_HangSpec_Trail
 emitters [
  {emitterName=FlareLoop003, rendererClasses [&NiagaraSpriteRendererProperties, &NiagaraLightRendererProperties]}
  {emitterName=Flames, rendererClasses [&NiagaraMeshRendererProperties, &NiagaraLightRendererProperties]}
  {emitterName=Embers002, rendererClasses [&NiagaraSpriteRendererProperties]}
  {emitterName=Speedlines, rendererClasses [&NiagaraSpriteRendererProperties]}
  {emitterName=FN_AnimTrail, rendererClasses [&NiagaraRibbonRendererProperties]}
 ]
 userVariables [
  User.bIsFrontEnd {type=NiagaraBool}
  User.bIsFrontEndPreview {type=NiagaraBool}
  User.bIsFullyDeployed {type=NiagaraBool}
  User.ForwardDot {type=NiagaraFloat, defaultValue=0, description=""}
  User.RibbonIDOffset {type=NiagaraInt32}
  User.RightDot {type=NiagaraFloat, defaultValue=0, description=""}
  User.SkeletalMesh {
   type=NiagaraDataInterfaceSkeletalMesh
   defaultValue {
    struct=NiagaraExt_VariableValue_DataInterface
    value {dataInterface=&self:NiagaraDataInterfaceSkeletalMesh_0}
   }
  }
  User.Variant {type=NiagaraFloat, defaultValue=0, description=""}
 ]
}
