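# Material function for the HangSpec glider that layers animated glowing dot and heart effects onto incoming emissive, masked by TechartMask channels.
# Dots tile from rotated pre-skinned local-position UVs, using per-cell random values and scrolling cellular noise for variation and two-color tint.
# Hearts take brightness from a panning tile texture.
# A time-driven sphere mask, offset by AnimationInput and a delay, grows along local Z as a deploy reveal.
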MaterialFunction /BRCosmetics/Gadgets/Assets/VinderTech_GliderChute/Glider_HangSpec/Materials/MF_HangSpec
graph {expressions {
  MaterialExpressionAdd_1 {
   class=MaterialExpressionAdd
   inputs [
    A {from=MaterialExpressionRotator_0}
    B {from=MaterialExpressionComponentMask_1}
   ]
  }
  MaterialExpressionAdd_2 {
   class=MaterialExpressionAdd
   inputs [
    A {from=MaterialExpressionScalarParameter_6}
    B {from=MaterialExpressionScalarParameter_19}
   ]
   properties {constB=2}
  }
  MaterialExpressionAdd_4 {
   class=MaterialExpressionAdd
   inputs [
    A {from=MaterialExpressionCeil_3}
    B {from=MaterialExpressionNamedRerouteUsage_25}
   ]
  }
  MaterialExpressionAdd_5 {
   class=MaterialExpressionAdd
   inputs [
    A {from=MaterialExpressionScalarParameter_10}
    B {from=MaterialExpressionNamedRerouteUsage_4}
   ]
  }
  MaterialExpressionAdd_7 {
   class=MaterialExpressionAdd
   inputs [
    A {from=MaterialExpressionComponentMask_4}
    B {from=MaterialExpressionConstant_9}
   ]
  }
  MaterialExpressionAppendVector_1 {
   class=MaterialExpressionAppendVector
   inputs [
    A {from=MaterialExpressionScalarParameter_5, output=""}
    B {from=MaterialExpressionScalarParameter_5, output=""}
   ]
  }
  MaterialExpressionCeil_1 {class=MaterialExpressionCeil, inputs [None {from=MaterialExpressionNamedRerouteUsage_21, output=""}], properties {}}
  MaterialExpressionCeil_3 {class=MaterialExpressionCeil, inputs [None {from=MaterialExpressionMaterialFunctionCall_3, output=Result}], properties {}}
  MaterialExpressionChannelMaskParameter_0 {
   class=MaterialExpressionChannelMaskParameter
   inputs [None {from=MaterialExpressionFunctionInput_1, output=""}]
   properties {parameterName=DotPatternEffectChannel}
  }
  MaterialExpressionChannelMaskParameter_1 {
   class=MaterialExpressionChannelMaskParameter
   inputs [None {from=MaterialExpressionFunctionInput_1, output=""}]
   properties {maskChannel=Green, parameterName=HeartPatternMaskChannel, defaultValue=<rgba 0 1 0 0>}
  }
  MaterialExpressionComponentMask_0 {class=MaterialExpressionComponentMask, inputs [None {from=MaterialExpressionNamedRerouteUsage_16}], properties {b=true, g=true}}
  MaterialExpressionComponentMask_1 {class=MaterialExpressionComponentMask, inputs [None {from=MaterialExpressionVectorParameter_1, output=RGB}], properties {g=true, r=true}}
  MaterialExpressionComponentMask_2 {class=MaterialExpressionComponentMask, inputs [None {from=MaterialExpressionFrac_1}], properties {r=true}}
  MaterialExpressionComponentMask_3 {class=MaterialExpressionComponentMask, inputs [None {from=MaterialExpressionNamedRerouteUsage_24}], properties {r=true}}
  MaterialExpressionComponentMask_4 {class=MaterialExpressionComponentMask, inputs [None {from=MaterialExpressionNamedRerouteUsage_17}], properties {b=true}}
  MaterialExpressionComponentMask_5 {class=MaterialExpressionComponentMask, inputs [None {from=MaterialExpressionNamedRerouteDeclaration_11}], properties {b=true}}
  MaterialExpressionConstantBiasScale_0 {class=MaterialExpressionConstantBiasScale, inputs [None {from=MaterialExpressionMaterialFunctionCall_5, output=Result}]}
  MaterialExpressionConstantBiasScale_1 {class=MaterialExpressionConstantBiasScale, inputs [None {from=MaterialExpressionMaterialFunctionCall_6, output=Result}]}
  MaterialExpressionConstant_0 {class=MaterialExpressionConstant, properties {r=0.5}}
  MaterialExpressionConstant_1 {class=MaterialExpressionConstant, properties {r=255}}
  MaterialExpressionConstant_2 {class=MaterialExpressionConstant, properties {r=255}}
  MaterialExpressionConstant_5 {class=MaterialExpressionConstant, properties {r=-10}}
  MaterialExpressionConstant_6 {class=MaterialExpressionConstant}
  MaterialExpressionConstant_9 {class=MaterialExpressionConstant, properties {r=28.5}}
  MaterialExpressionConvert_0 {
   class=MaterialExpressionConvert
   inputs [
    None {from=MaterialExpressionComponentMask_3}
    None {from=""}
   ]
   properties {
    nodeName=Make Float2
    convertInputs [
     {
      type=Scalar
      expressionInput {
       expression=&self:MaterialExpressionComponentMask_3
       inputName=None
       mask=0
       maskA=0
       maskB=0
       maskG=0
       maskR=0
       outputIndex=0
      }
     }
     {type=Scalar, expressionInput {expression=null, inputName=None, mask=0, maskA=0, maskB=0, maskG=0, maskR=0, outputIndex=0}}
    ]
    convertOutputs [{type=Vector2}]
   }
  }
  MaterialExpressionDesaturation_0 {
   class=MaterialExpressionDesaturation
   inputs [
    None {from=MaterialExpressionNamedRerouteUsage_28, output=""}
    Fraction {from=MaterialExpressionScalarParameter_2, output=""}
   ]
   properties {}
  }
  MaterialExpressionDivide_0 {
   class=MaterialExpressionDivide
   inputs [
    A {from=MaterialExpressionSubtract_0}
    B {from=MaterialExpressionScalarParameter_7}
   ]
  }
  MaterialExpressionDivide_1 {
   class=MaterialExpressionDivide
   inputs [
    A {from=MaterialExpressionAdd_4}
    B {from=MaterialExpressionScalarParameter_9}
   ]
  }
  MaterialExpressionFrac_0 {class=MaterialExpressionFrac, inputs [None {from=MaterialExpressionNamedRerouteUsage_19, output=""}], properties {}}
  MaterialExpressionFrac_1 {class=MaterialExpressionFrac, inputs [None {from=MaterialExpressionMultiply_12, output=""}], properties {}}
  MaterialExpressionFunctionInput_0 {class=MaterialExpressionFunctionInput, inputs [Preview {from="", output=""}], properties {inputType=FunctionInput_MaterialAttributes}}
  MaterialExpressionFunctionInput_1 {class=MaterialExpressionFunctionInput, inputs [Preview {from="", output=""}], properties {inputName=TechartMask}}
  MaterialExpressionFunctionOutput_0 {class=MaterialExpressionFunctionOutput, inputs [None {from=MaterialExpressionSetMaterialAttributes_1, output=""}]}
  MaterialExpressionGetMaterialAttributes_0 {
   class=MaterialExpressionGetMaterialAttributes
   inputs [None {from=MaterialExpressionFunctionInput_0, output=""}]
   properties {attributeGetTypes [#1]}
  }
  MaterialExpressionLinearInterpolate_1 {
   class=MaterialExpressionLinearInterpolate
   inputs [
    A {from=MaterialExpressionScalarParameter_1}
    B {from=MaterialExpressionScalarParameter_0}
    Alpha {from=MaterialExpressionTextureSample_4, output=RGB}
   ]
  }
  MaterialExpressionLinearInterpolate_4 {
   class=MaterialExpressionLinearInterpolate
   inputs [
    A {from=MaterialExpressionAdd_5}
    B {from=MaterialExpressionMultiply_21}
    Alpha {from=MaterialExpressionNamedRerouteUsage_10}
   ]
  }
  MaterialExpressionLinearInterpolate_5 {
   class=MaterialExpressionLinearInterpolate
   inputs [
    A {from=MaterialExpressionLinearInterpolate_7}
    B {from=MaterialExpressionNamedRerouteUsage_11}
    Alpha {from=MaterialExpressionNamedRerouteUsage_12}
   ]
  }
  MaterialExpressionLinearInterpolate_7 {
   class=MaterialExpressionLinearInterpolate
   inputs [
    A {from=MaterialExpressionNamedRerouteDeclaration_16}
    B {from=MaterialExpressionNamedRerouteUsage_3}
    Alpha {from=MaterialExpressionNamedRerouteUsage_13}
   ]
  }
  MaterialExpressionLinearInterpolate_8 {
   class=MaterialExpressionLinearInterpolate
   inputs [
    A {from=MaterialExpressionVectorParameter_0, output=RGB}
    B {from=MaterialExpressionVectorParameter_2, output=RGB}
    Alpha {from=MaterialExpressionSaturate_2}
   ]
  }
  MaterialExpressionLocalPosition_0 {class=MaterialExpressionLocalPosition, inputs [], properties {localOrigin=InstancePreSkinning}}
  MaterialExpressionMaterialFunctionCall_3 {
   class=MaterialExpressionMaterialFunctionCall
   inputs [
    X {from=MaterialExpressionConstant_6}
    Y {from=MaterialExpressionMultiply_16}
   ]
   properties {materialFunction=&MakeFloat2}
  }
  MaterialExpressionMaterialFunctionCall_5 {
   class=MaterialExpressionMaterialFunctionCall
   inputs [
    NoisePower {from=MaterialExpressionConstant_1}
    Coords {from=MaterialExpressionNamedRerouteDeclaration_15}
   ]
   properties {materialFunction=/Game/Global/MaterialFunctions/TechArt/MF_RandomNoise}
  }
  MaterialExpressionMaterialFunctionCall_6 {
   class=MaterialExpressionMaterialFunctionCall
   inputs [
    NoisePower {from=MaterialExpressionConstant_2}
    Coords {from=MaterialExpressionConvert_0}
   ]
   properties {materialFunction=&MF_RandomNoise}
  }
  MaterialExpressionMultiply_11 {
   class=MaterialExpressionMultiply
   inputs [
    A {from=MaterialExpressionAdd_1}
    B {from=MaterialExpressionScalarParameter_8}
   ]
   properties {constB=0.1}
  }
  MaterialExpressionMultiply_12 {
   class=MaterialExpressionMultiply
   inputs [
    A {from=MaterialExpressionNamedRerouteUsage_20}
    B {from=""}
   ]
   properties {constB=0.5}
  }
  MaterialExpressionMultiply_13 {
   class=MaterialExpressionMultiply
   inputs [
    A {from=MaterialExpressionStep_2}
    B {from=MaterialExpressionPower_3}
   ]
  }
  MaterialExpressionMultiply_14 {
   class=MaterialExpressionMultiply
   inputs [
    A {from=MaterialExpressionNamedRerouteDeclaration_17}
    B {from=MaterialExpressionMultiply_13}
   ]
  }
  MaterialExpressionMultiply_16 {
   class=MaterialExpressionMultiply
   inputs [
    A {from=MaterialExpressionTime_1}
    B {from=MaterialExpressionLinearInterpolate_4}
   ]
  }
  MaterialExpressionMultiply_17 {
   class=MaterialExpressionMultiply
   inputs [
    A {from=MaterialExpressionNamedRerouteUsage_22}
    B {from=MaterialExpressionMultiply_14}
   ]
  }
  MaterialExpressionMultiply_19 {
   class=MaterialExpressionMultiply
   inputs [
    A {from=MaterialExpressionConstantBiasScale_1}
    B {from=MaterialExpressionScalarParameter_11}
   ]
  }
  MaterialExpressionMultiply_21 {
   class=MaterialExpressionMultiply
   inputs [
    A {from=MaterialExpressionAdd_5}
    B {from=""}
   ]
   properties {constB=-1}
  }
  MaterialExpressionMultiply_26 {
   class=MaterialExpressionMultiply
   inputs [
    A {from=MaterialExpressionMultiply_17}
    B {from=MaterialExpressionLinearInterpolate_8}
   ]
  }
  MaterialExpressionMultiply_27 {
   class=MaterialExpressionMultiply
   inputs [
    A {from=MaterialExpressionMultiply_26}
    B {from=MaterialExpressionScalarParameter_14}
   ]
  }
  MaterialExpressionMultiply_28 {
   class=MaterialExpressionMultiply
   inputs [
    A {from=MaterialExpressionMultiply_27}
    B {from=MaterialExpressionNamedRerouteUsage_26}
   ]
  }
  MaterialExpressionMultiply_29 {
   class=MaterialExpressionMultiply
   inputs [
    A {from=MaterialExpressionMultiply_4}
    B {from=MaterialExpressionNamedRerouteUsage_27}
   ]
  }
  MaterialExpressionMultiply_4 {
   class=MaterialExpressionMultiply
   inputs [
    A {from=MaterialExpressionDesaturation_0}
    B {from=MaterialExpressionLinearInterpolate_1}
   ]
  }
  MaterialExpressionMultiply_7 {
   class=MaterialExpressionMultiply
   inputs [
    A {from=MaterialExpressionSaturate_0}
    B {from=""}
   ]
   properties {constB=450}
  }
  MaterialExpressionNamedRerouteDeclaration_0 {
   class=MaterialExpressionNamedRerouteDeclaration
   inputs [None {from=MaterialExpressionChannelMaskParameter_0, output=R}]
   properties {name=DotPatternMask, nodeColor=<rgba 0.4297 1 0 1>}
  }
  MaterialExpressionNamedRerouteDeclaration_10 {
   class=MaterialExpressionNamedRerouteDeclaration
   inputs [None {from=MaterialExpressionSphereMask_0}]
   properties {name=AnimatedMask, nodeColor=<rgba 1 0 0.5156 1>}
  }
  MaterialExpressionNamedRerouteDeclaration_11 {
   class=MaterialExpressionNamedRerouteDeclaration
   inputs [None {from=MaterialExpressionVertexInterpolator_0, output=PS}]
   properties {name=PreskinnedLocalPos, nodeColor=<rgba 0 0.9766 1 1>}
  }
  MaterialExpressionNamedRerouteDeclaration_13 {
   class=MaterialExpressionNamedRerouteDeclaration
   inputs [None {from=MaterialExpressionMultiply_11}]
   properties {name=DotPatternUVs, nodeColor=<rgba 0.4297 1 0 1>}
  }
  MaterialExpressionNamedRerouteDeclaration_14 {
   class=MaterialExpressionNamedRerouteDeclaration
   inputs [None {from=MaterialExpressionPower_2}]
   properties {name=DotPatternAnimatedNoise, nodeColor=<rgba 1 0 0.75 1>}
  }
  MaterialExpressionNamedRerouteDeclaration_15 {
   class=MaterialExpressionNamedRerouteDeclaration
   inputs [None {from=MaterialExpressionCeil_1}]
   properties {name=UVCeil, nodeColor=<rgba 0 1 0.9297 1>}
  }
  MaterialExpressionNamedRerouteDeclaration_16 {
   class=MaterialExpressionNamedRerouteDeclaration
   inputs [None {from=MaterialExpressionGetMaterialAttributes_0, output=EmissiveColor}]
   properties {name=IN_Emissive, nodeColor=<rgba 1 0.1172 0 1>}
  }
  MaterialExpressionNamedRerouteDeclaration_17 {
   class=MaterialExpressionNamedRerouteDeclaration
   inputs [None {from=MaterialExpressionConstantBiasScale_0}]
   properties {name=DotPatternRandom, nodeColor=<rgba 1 0.0938 0 1>}
  }
  MaterialExpressionNamedRerouteDeclaration_3 {
   class=MaterialExpressionNamedRerouteDeclaration
   inputs [None {from=MaterialExpressionMultiply_28}]
   properties {name=DotPattern, nodeColor=<rgba 0 1 0.3438 1>}
  }
  MaterialExpressionNamedRerouteDeclaration_5 {
   class=MaterialExpressionNamedRerouteDeclaration
   inputs [None {from=MaterialExpressionMultiply_19}]
   properties {name=RandomSpeed, nodeColor=<rgba 0.8984 0 1 1>}
  }
  MaterialExpressionNamedRerouteDeclaration_7 {
   class=MaterialExpressionNamedRerouteDeclaration
   inputs [None {from=MaterialExpressionStep_3}]
   properties {name=DotPatternZMask, nodeColor=<rgba 1 0.4219 0 1>}
  }
  MaterialExpressionNamedRerouteDeclaration_8 {
   class=MaterialExpressionNamedRerouteDeclaration
   inputs [None {from=MaterialExpressionChannelMaskParameter_1, output=R}]
   properties {name=HeartMask, nodeColor=<rgba 0.6406 0 1 1>}
  }
  MaterialExpressionNamedRerouteDeclaration_9 {
   class=MaterialExpressionNamedRerouteDeclaration
   inputs [None {from=MaterialExpressionMultiply_29}]
   properties {name=Hearts, nodeColor=<rgba 0.6406 1 0 1>}
  }
  MaterialExpressionNamedRerouteUsage_10 {class=MaterialExpressionNamedRerouteUsage}
  MaterialExpressionNamedRerouteUsage_11 {class=MaterialExpressionNamedRerouteUsage}
  MaterialExpressionNamedRerouteUsage_12 {class=MaterialExpressionNamedRerouteUsage}
  MaterialExpressionNamedRerouteUsage_13 {class=MaterialExpressionNamedRerouteUsage}
  MaterialExpressionNamedRerouteUsage_16 {class=MaterialExpressionNamedRerouteUsage}
  MaterialExpressionNamedRerouteUsage_17 {class=MaterialExpressionNamedRerouteUsage}
  MaterialExpressionNamedRerouteUsage_19 {class=MaterialExpressionNamedRerouteUsage}
  MaterialExpressionNamedRerouteUsage_20 {class=MaterialExpressionNamedRerouteUsage}
  MaterialExpressionNamedRerouteUsage_21 {class=MaterialExpressionNamedRerouteUsage}
  MaterialExpressionNamedRerouteUsage_22 {class=MaterialExpressionNamedRerouteUsage}
  MaterialExpressionNamedRerouteUsage_24 {class=MaterialExpressionNamedRerouteUsage}
  MaterialExpressionNamedRerouteUsage_25 {class=MaterialExpressionNamedRerouteUsage}
  MaterialExpressionNamedRerouteUsage_26 {class=MaterialExpressionNamedRerouteUsage}
  MaterialExpressionNamedRerouteUsage_27 {class=MaterialExpressionNamedRerouteUsage}
  MaterialExpressionNamedRerouteUsage_28 {class=MaterialExpressionNamedRerouteUsage}
  MaterialExpressionNamedRerouteUsage_29 {class=MaterialExpressionNamedRerouteUsage}
  MaterialExpressionNamedRerouteUsage_3 {class=MaterialExpressionNamedRerouteUsage}
  MaterialExpressionNamedRerouteUsage_4 {class=MaterialExpressionNamedRerouteUsage}
  MaterialExpressionPanner_2 {
   class=MaterialExpressionPanner
   inputs [
    Coordinate {from=""}
    Time {from=""}
    Speed {from=MaterialExpressionAppendVector_1}
   ]
   properties {speedY=0.1}
  }
  MaterialExpressionPower_2 {
   class=MaterialExpressionPower
   inputs [
    Base {from=MaterialExpressionTextureSample_3, output=R}
    Exp {from=""}
   ]
   properties {constExponent=4}
  }
  MaterialExpressionPower_3 {
   class=MaterialExpressionPower
   inputs [
    Base {from=MaterialExpressionSphereMask_1}
    Exp {from=""}
   ]
  }
  MaterialExpressionRotator_0 {
   class=MaterialExpressionRotator
   inputs [
    Coordinate {from=MaterialExpressionComponentMask_0, output=""}
    Time {from=MaterialExpressionScalarParameter_13, output=""}
   ]
   properties {}
  }
  MaterialExpressionSaturate_0 {class=MaterialExpressionSaturate, inputs [None {from=MaterialExpressionDivide_0}]}
  MaterialExpressionSaturate_2 {class=MaterialExpressionSaturate, inputs [None {from=MaterialExpressionNamedRerouteUsage_29}]}
  MaterialExpressionScalarParameter_0 {class=MaterialExpressionScalarParameter, properties {defaultValue=10, parameterName=HeartsBrightnessMax}}
  MaterialExpressionScalarParameter_1 {class=MaterialExpressionScalarParameter, properties {defaultValue=1, parameterName=HeartsBrightnessMin}}
  MaterialExpressionScalarParameter_10 {class=MaterialExpressionScalarParameter, properties {defaultValue=2, parameterName=Speed}}
  MaterialExpressionScalarParameter_11 {class=MaterialExpressionScalarParameter, properties {defaultValue=10, parameterName=DotPatternRandomSpeed}}
  MaterialExpressionScalarParameter_13 {class=MaterialExpressionScalarParameter, properties {defaultValue=-0.19, parameterName=DotPatternUVRotation}}
  MaterialExpressionScalarParameter_14 {class=MaterialExpressionScalarParameter, properties {defaultValue=4.5, parameterName=DotPatternIntensity}}
  MaterialExpressionScalarParameter_15 {class=MaterialExpressionScalarParameter, properties {defaultValue=0.4, parameterName=DotPatternHardness}}
  MaterialExpressionScalarParameter_18 {class=MaterialExpressionScalarParameter, properties {defaultValue=0.6, parameterName=DotPatternRadius}}
  MaterialExpressionScalarParameter_19 {class=MaterialExpressionScalarParameter, properties {defaultValue=1.15, parameterName=DeployAnimationDelay}}
  MaterialExpressionScalarParameter_2 {class=MaterialExpressionScalarParameter, properties {defaultValue=-1, parameterName=HeartDotsDesatuation}}
  MaterialExpressionScalarParameter_5 {class=MaterialExpressionScalarParameter, properties {defaultValue=0.1, parameterName=HeartsPanningSpeed}}
  MaterialExpressionScalarParameter_6 {class=MaterialExpressionScalarParameter, properties {parameterName=AnimationInput}}
  MaterialExpressionScalarParameter_7 {class=MaterialExpressionScalarParameter, properties {defaultValue=2, parameterName=DeployAnimationDuration}}
  MaterialExpressionScalarParameter_8 {class=MaterialExpressionScalarParameter, properties {defaultValue=0.25, parameterName=DotPatternUVScale}}
  MaterialExpressionScalarParameter_9 {class=MaterialExpressionScalarParameter, properties {defaultValue=30, parameterName=DotPatternNoiseUVScale}}
  MaterialExpressionSetMaterialAttributes_1 {
   class=MaterialExpressionSetMaterialAttributes
   inputs [
    MaterialAttributes {from=MaterialExpressionGetMaterialAttributes_0, output=MaterialAttributes}
    EmissiveColor {from=MaterialExpressionLinearInterpolate_5, output=""}
   ]
   properties {attributeSetTypes [#1]}
  }
  MaterialExpressionSphereMask_0 {
   class=MaterialExpressionSphereMask
   inputs [
    A {from=MaterialExpressionComponentMask_5, output=""}
    B {from=MaterialExpressionConstant_5, output=""}
    Radius {from=MaterialExpressionMultiply_7, output=""}
    Hardness {from="", output=""}
   ]
   properties {hardnessPercent=0}
  }
  MaterialExpressionSphereMask_1 {
   class=MaterialExpressionSphereMask
   inputs [
    A {from=MaterialExpressionFrac_0, output=""}
    B {from=MaterialExpressionConstant_0, output=""}
    Radius {from=MaterialExpressionScalarParameter_18, output=""}
    Hardness {from=MaterialExpressionScalarParameter_15, output=""}
   ]
   properties {attenuationRadius=0.45}
  }
  MaterialExpressionStep_2 {
   class=MaterialExpressionStep
   inputs [
    Y {from="", output=""}
    X {from=MaterialExpressionComponentMask_2, output=""}
   ]
   properties {constY=0.5}
  }
  MaterialExpressionStep_3 {
   class=MaterialExpressionStep
   inputs [
    Y {from="", output=""}
    X {from=MaterialExpressionAdd_7, output=""}
   ]
   properties {}
  }
  MaterialExpressionSubtract_0 {
   class=MaterialExpressionSubtract
   inputs [
    A {from=MaterialExpressionTime_0}
    B {from=MaterialExpressionAdd_2}
   ]
  }
  MaterialExpressionTextureSample_3 {
   class=MaterialExpressionTextureSample
   inputs [
    UVs {from=MaterialExpressionDivide_1}
    Tex {from=""}
    Apply View MipBias {from=""}
   ]
   properties {samplerType=SAMPLERTYPE_LinearGrayscale, texture=/BRCosmetics/Effects/Fort_Effects/BluGlow/Violentium/Cellular_basecolor}
  }
  MaterialExpressionTextureSample_4 {
   class=MaterialExpressionTextureSample
   inputs [
    UVs {from=MaterialExpressionPanner_2}
    Tex {from=""}
    Apply View MipBias {from=""}
   ]
   properties {
    samplerType=SAMPLERTYPE_Masks
    texture=/BRCosmetics/Gadgets/Assets/VinderTech_GliderChute/Glider_HangSpec/Textures/T_heartspec_glider_hearts_tile
   }
  }
  MaterialExpressionTime_0 {class=MaterialExpressionTime}
  MaterialExpressionTime_1 {class=MaterialExpressionTime}
  MaterialExpressionVectorParameter_0 {class=MaterialExpressionVectorParameter, properties {parameterName=DotPatternColorA, defaultValue=<rgba 0 10 7 1>}}
  MaterialExpressionVectorParameter_1 {class=MaterialExpressionVectorParameter, properties {parameterName=DotPatternUVOffset, defaultValue=<rgba 22 0 0 1>}}
  MaterialExpressionVectorParameter_2 {class=MaterialExpressionVectorParameter, properties {parameterName=DotPatternColorB, defaultValue=<rgba 0 6 10 1>}}
  MaterialExpressionVertexInterpolator_0 {class=MaterialExpressionVertexInterpolator, inputs [VS {from=MaterialExpressionLocalPosition_0, output=XYZ}], properties {}}
 }}
